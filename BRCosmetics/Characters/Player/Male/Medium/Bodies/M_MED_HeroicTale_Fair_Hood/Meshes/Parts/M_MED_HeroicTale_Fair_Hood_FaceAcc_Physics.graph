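# Physics asset for the face-accessory part of the male medium HeroicTale Fair Hood cosmetic, defining physics bodies for the back of the hood.
# Three capsule bodies form a chain from a small base anchor through the main back-hood bone to the hood tip.
# Two constraints link each bone to the one before it with narrow angular limits, restricting how far the hood can bend and twist.
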PhysicsAsset /BRCosmetics/Characters/Player/Male/Medium/Bodies/M_MED_HeroicTale_Fair_Hood/Meshes/Parts/M_MED_HeroicTale_Fair_Hood_FaceAcc_Physics
bodies {
 base_back_hood [{shapeType=Capsule, length=1, radius=0.5, shapeName=base_front_hood_capsule}]
 dyn_back_hood [{shapeType=Capsule, length=2.5632, radius=3.1268, shapeName=dyn_back_hood_capsule, center=<4.044 0 0>, rotation=<rot 90 0 0>}]
 dyn_back_hood_tip [{shapeType=Capsule, length=1.6076, radius=1.0626, shapeName=dyn_back_hood_capsule, center=<1.9423 0 0>, rotation=<rot 90 0 0>}]
}
constraints [
 {bone1Name=dyn_back_hood, bone2Name=base_back_hood, swing1LimitDegrees=25, swing2LimitDegrees=20, twistLimitDegrees=0}
 {bone1Name=dyn_back_hood_tip, bone2Name=dyn_back_hood, swing1LimitDegrees=0, swing2LimitDegrees=10, twistLimitDegrees=10}
]
